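# Unlit translucent material for lightning on Niagara mesh particles in the Static Shades character skin FX.
# A sinusoidal gradient mask along UV V, optionally reversed, drives hard-edged opacity, depth-faded above low quality, and normal-based width offsets.
# Vector noise and a time-scrolled sine add jagged wobble, plus a camera-ward offset.
# Particle color is hue-shifted along the gradient and boosted for emissive, with dynamic parameters controlling shape, motion and color.
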Material /Game/Characters/Player/Male/Medium/Bodies/M_MED_StaticShades/FX/Materials/M_Skin_StaticShades_Lightning
graph {
 properties {
  bUsedWithNiagaraMeshParticles=true
  blendMode=BLEND_Translucent
  pixelDepthOffsetMode=PDOM_Legacy
  shadingModel=MSM_Unlit
  usedShadingModels=Unlit
  lightmassSettings {
   bCastShadowAsMasked=false
   bOverrideCastShadowAsMasked=false
   bOverrideDiffuseBoost=false
   bOverrideEmissiveBoost=false
   bOverrideExportResolutionScale=false
   diffuseBoost=999999
   emissiveBoost=1
   exportResolutionScale=16
  }
 }
 nodes [
  n1 = ComponentMask(None=TextureCoordinate()) {g=true}
  n2 = VertexNormalWS()
  n3 = Constant() {r=1}
  n4 = DynamicParameter() {paramNames [Width Max, Width Min, Noise Speed, Noise Intensity]}
  n5 = Ceil(None=ParticleRelativeTime() {desc=">0"})
  n6 = ComponentMask(None=TextureCoordinate()) {g=true}
  n7 = DynamicParameter() {parameterIndex=2, paramNames [Time of Day Color Multiplier, Vertex Shader Camera Offset, Depth Fade Distance, Reverse Lightning Direction]}
  n8 = LinearInterpolate(A=@n6, B=OneMinus(None=@n6), Alpha=Saturate(None=@n7.Reverse Lightning Direction))
  n9 = Time()
  n10 = DynamicParameter() {parameterIndex=1, paramNames [Mask Size, Mask Position, Color Boosting Intensity, Color Boosting Influience]}
  n11 = MaterialFunctionCall!(
   Gradient=@n8
   Mask Position=LinearInterpolate(A=Frac(None=@n9), B=@n10.Mask Position, Alpha=@n5) {constA=0.5}
   Mask Size=LinearInterpolate(A=Constant() {r=1}, B=@n10.Mask Size, Alpha=@n5)
  ) {materialFunction=/Game/Effects/Fort_Effects/MaterialFunctions/MF_GradientMask}
  n12 = MaterialFunctionCall() {materialFunction=&LocalPosition}
  n13 = VectorNoise(World Position=Add(A=Floor(None=@n12.Instance Local Position)) {constB=0.5})
  n14 = Add(A=ComponentMask(None=@n13) {r=true}, B=ParticleRandom())
  n15 = DynamicParameter() {paramNames [Width Max, Width Min, Noise Speed, Noise Intensity]}
  n16 = Multiply(
   A=Multiply(A=@n11.Sinusoidal, B=LinearInterpolate(A=@n3, B=@n4.Width Max, Alpha=@n5))
   B=Max(A=Frac(None=@n14), B=@n15.Width Min) {constB=0.4}
  )
  n17 = Normalize(VectorInput=Subtract(A=@n13) {constB=0.5})
  n18 = Multiply(A=Time(), B=@n15.Noise Speed) {constB=6}
  n19 = Multiply(A=Sine(None=Add(A=@n18, B=@n1)), B=@n15.Noise Intensity)
  n20 = Multiply(A=Ceil(None=Frac(None=@n1)), B=Add(A=Multiply(A=@n2, B=@n16), B=Multiply(A=@n17, B=@n19)))
  n21 = Subtract(A=CameraPositionWS(), B=Add(A=@n20, B=WorldPosition().XYZ))
  n22 = DynamicParameter() {parameterIndex=2, paramNames [Time of Day Color Multiplier, Vertex Shader Camera Offset, Depth Fade Distance, Reverse Lightning Direction]}
  Add(A=@n20, B=Multiply(A=Normalize(VectorInput=@n21), B=@n22.Vertex Shader Camera Offset))
  n23 = ParticleColor()
  n24 = Add(A=Multiply(A=Subtract(A=@n8) {constB=0.5}, B=ScalarParameter() {defaultValue=3, parameterName=Hue shift contrast})) {constB=0.5}
  n25 = MaterialFunctionCall(Hue Shift Percentage=Multiply(A=@n23.A, B=Saturate(None=@n24)), Texture=@n23.RGB) {materialFunction=&HueShift}
  n26 = DynamicParameter() {parameterIndex=1, paramNames [Mask Size, Mask Position, Color Boosting Intensity, Color Boosting Influience]}
  n27 = Multiply(A=@n26.Color Boosting Intensity, B=@n25.Result)
  n28 = DynamicParameter() {parameterIndex=2, paramNames [Time of Day Color Multiplier, Vertex Shader Camera Offset, Depth Fade Distance, Reverse Lightning Direction]}
  MaterialFunctionCall(Emissive=Multiply(A=LinearInterpolate(A=@n25.Result, B=@n27, Alpha=@n26.Color Boosting Influience), B=@n28.Time of Day Color Multiplier)) {materialFunction=&MF_FX_EmissiveAdjustments}
  Max(A=@n26.Color Boosting Intensity)
  n29 = Multiply(A=Saturate(None=Ceil(None=@n15.Width Max)), B=Ceil(None=@n11.Sinusoidal))
  QualitySwitch(Default=DepthFade(Opacity=@n29, FadeDistance=@n28.Depth Fade Distance), Low=@n29)
 ]
}
