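# Translucent unlit material for lightning bits in the StaticShades skin effects, enabled for Niagara mesh particles.
# A moving gradient mask along UV shapes a camera-facing strip's width and gates opacity, with depth fade outside Low quality.
# Per-cell vector noise displaces the strip along its local axis; both width lerp inputs read Width Min.
# Particle color is hue-shifted over lifetime, then boosted and scaled by a time-of-day multiplier for emissive.
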Material /Game/Characters/Player/Male/Medium/Bodies/M_MED_StaticShades/FX/Materials/M_Skin_StaticShades_Lightning_Bits
graph {
 properties {
  bUsedWithNiagaraMeshParticles=true
  blendMode=BLEND_Translucent
  pixelDepthOffsetMode=PDOM_Legacy
  shadingModel=MSM_Unlit
  twoSided=true
  usedShadingModels=Unlit
  lightmassSettings {
   bCastShadowAsMasked=false
   bOverrideCastShadowAsMasked=false
   bOverrideDiffuseBoost=false
   bOverrideEmissiveBoost=false
   bOverrideExportResolutionScale=false
   diffuseBoost=999999
   emissiveBoost=1
   exportResolutionScale=16
  }
 }
 nodes [
  n1 = WorldPosition()
  n2 = Normalize(VectorInput=Subtract(A=@n1.XYZ, B=CameraPositionWS()))
  n3 = Transform(None=Constant3Vector() {constant=<rgba 0 1 0 0>})
  n4 = TextureCoordinate()
  n5 = Multiply(A=Normalize(VectorInput=CrossProduct(A=@n2, B=@n3)), B=Subtract(A=ComponentMask(None=@n4) {r=true}) {constB=0.5})
  n6 = ComponentMask(None=TextureCoordinate()) {g=true}
  n7 = DynamicParameter() {parameterIndex=2, paramNames [Time of Day Color Multiplier, Vertex Shader Camera Offset, Depth Fade Distance, Reverse Lightning Direction]}
  n8 = LinearInterpolate(A=@n6, B=OneMinus(None=@n6), Alpha=Saturate(None=@n7.Reverse Lightning Direction))
  n9 = Multiply(A=@n8, B=ScalarParameter() {parameterName=Gradient Size})
  n10 = DynamicParameter() {parameterIndex=1, paramNames [Mask Size, Mask Position, Color Boosting Intensity, Color Boosting Influience]}
  n11 = Ceil(None=ParticleRelativeTime() {desc=">0"})
  n12 = LinearInterpolate(A=Frac(None=Time()), B=@n10.Mask Position, Alpha=@n11) {constA=0.5}
  n13 = MaterialFunctionCall!(Gradient=@n9, Mask Position=@n12, Mask Size=Multiply(A=ParticleRandom(), B=ScalarParameter() {defaultValue=3, parameterName=mask siz})) {materialFunction=/Game/Effects/Fort_Effects/MaterialFunctions/MF_GradientMask}
  n14 = SquareRoot(None=@n13.Sinusoidal)
  n15 = DynamicParameter() {paramNames [Width Max, Width Min, Noise Speed, Noise Intensity]}
  n16 = TransformPosition(None=@n1.XYZ) {transformSourceType=TRANSFORMPOSSOURCE_World, transformType=TRANSFORMPOSSOURCE_Instance}
  n17 = VectorNoise(World Position=Add(A=Floor(None=@n16)) {constB=0.5})
  n18 = ComponentMask(None=@n17) {r=true}
  n19 = ConstantBiasScale(None=SquareRoot(None=@n12)) {bias=-0.5, scale=2}
  n20 = Multiply(A=Subtract(A=ComponentMask(None=@n17) {g=true}) {constB=0.5}, B=ScalarParameter() {defaultValue=5, parameterName=Tangent displacement})
  Add(
   A=Multiply(A=Multiply(A=@n5, B=@n14), B=LinearInterpolate(A=@n15.Width Min, B=@n15.Width Min, Alpha=@n18))
   B=Multiply(A=Multiply(A=@n19, B=@n20), B=Normalize(VectorInput=@n3))
  ) {constB=0}
  Add(A=Add(A=@n9) {constB=0.45}, B=Subtract(A=ComponentMask(None=@n17) {g=true}) {constB=0.5})
  LinearInterpolate(A=Constant() {r=1}, B=@n10.Mask Size, Alpha=@n11)
  n21 = ParticleColor()
  n22 = ParticleRelativeTime()
  n23 = MaterialFunctionCall(Hue Shift Percentage=Multiply(A=@n21.A, B=SquareRoot(None=@n22)), Texture=@n21.RGB) {materialFunction=&HueShift}
  n24 = DynamicParameter() {parameterIndex=1, paramNames [Mask Size, Mask Position, Color Boosting Intensity, Color Boosting Influience]}
  n25 = Multiply(A=@n24.Color Boosting Intensity, B=@n23.Result)
  n26 = DynamicParameter() {parameterIndex=2, paramNames [Time of Day Color Multiplier, Vertex Shader Camera Offset, Depth Fade Distance, Reverse Lightning Direction]}
  MaterialFunctionCall(Emissive=Multiply(A=LinearInterpolate(A=@n23.Result, B=@n25, Alpha=@n24.Color Boosting Influience), B=@n26.Time of Day Color Multiplier)) {materialFunction=&MF_FX_EmissiveAdjustments}
  Normalize(VectorInput=@n16)
  n27 = DynamicParameter() {paramNames [Width Max, Width Min, Noise Speed, Noise Intensity]}
  n28 = Multiply(A=Saturate(None=Ceil(None=@n27.Width Max)), B=Ceil(None=@n13.Sinusoidal))
  QualitySwitch(Default=DepthFade(Opacity=@n28, FadeDistance=@n26.Depth Fade Distance), Low=@n28)
  n29 = Subtract(A=@n8) {constB=0.5}
  n30 = ScalarParameter() {defaultValue=3, parameterName=Hue shift contrast}
  Saturate(None=Add(A=Multiply(A=@n29, B=@n30)) {constB=0.5})
  ScalarParameter() {defaultValue=1, parameterName=debugwidth}
  SquareRoot()
  VertexTangentWS()
  WorldPosition()
 ]
}
