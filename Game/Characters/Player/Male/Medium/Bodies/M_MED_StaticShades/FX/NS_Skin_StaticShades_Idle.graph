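# Niagara idle effect for the StaticShades player skin, drawing burst-spawned lightning meshes on the character's skeleton.
# User arrays pick each particle's body partition, side and lifetime, which determine the sampled bone and mirrored scale.
# Random travel/direction flags and lifetime curves drive dynamic material parameters across four mesh renderers.
# Hidden head/hand flags set visibility and bit spawning, overwriting an earlier camera-distance VisibilityTag.
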NiagaraSystem /Game/Characters/Player/Male/Medium/Bodies/M_MED_StaticShades/FX/NS_Skin_StaticShades_Idle
dependencies {
 usedDataInterfaces [&NiagaraDataInterfaceCurve, &NiagaraDataInterfaceCamera]
 usedDynamicInputs [&FNTimeOfDayScalar]
 usedModules [
  &self:SystemSpawnScript.SystemScriptSource.SystemScriptGraph.NiagaraNodeAssignment_0.SetVariables_#1
  &SystemState
 ]
}
emitters {StaticShadesLightning {
  emitterUpdateScript [
   &EmitterState()
   &SpawnBurst_Instantaneous(Use Spawn Probability=false, UseLoopCountLimit=false)
  ]
  particleSpawnScript [
   /SetVariables(Particles.PartitionParticles.Partition:NiagaraInt32=&SelectIntFromArray(
     Direct Array Index=&ReturnParticleIDAttributes()
     Direct Array Mode=::NewEnumerator1
     Int Selection Array=$User.LightningIndex
    )) {isSetParameters=true}
   /SetVariables(Particles.PartitionParticles.Partition=12) {enabled=false, isSetParameters=true}
   /SetVariables(
    Particles.IsHead:NiagaraBool=&SetBoolByIntComparison(A=$Particles.PartitionParticles.Partition, B=9, Comparison Type=::NewEnumerator0)
    Particles.IsLeftSide:NiagaraBool=&SetBoolByIntComparison(
     A=&SelectIntFromArray(
      Direct Array Index=&ReturnParticleIDAttributes()
      Direct Array Mode=::NewEnumerator1
      Int Selection Array=$User.SideIndex
     )
     B=0
     Comparison Type=::NewEnumerator0
    )
    Particles.IsLowerArm:NiagaraBool=&SetBoolByIntComparison(A=$Particles.PartitionParticles.Partition, B=5, Comparison Type=::NewEnumerator6)
    Particles.IsReverseDirection:NiagaraBool=&RandomBool()
    Particles.IsTravelingLightning:NiagaraBool=&RandomBool()
    Particles.SpawnBits:NiagaraInt32=&MakeCustomIntFromBool(Bool=&RandomBool(), True INT=0, False INT=1)
   ) {isSetParameters=true}
   /SetVariables(Particles.BoneSampleIndex:NiagaraInt32=&MakeCustomIntFromBool(
     Bool=$Particles.IsHead
     True INT=4
     False INT=&MakeCustomIntFromBool(
      Bool=$Particles.IsLeftSide
      True INT=&MakeCustomIntFromBool(Bool=$Particles.IsLowerArm, True INT=0, False INT=1)
      False INT=&MakeCustomIntFromBool(Bool=$Particles.IsLowerArm, True INT=2, False INT=3)
     )
    )) {isSetParameters=true}
   &DynamicMaterialParameters(
    Index 0 Param 1=2
    Index 0 Param 2=0.3
    Index 0 Param 3=2.5
    Index 0 Param 4=0.8
    Index 1 Param 1=&MakeCustomFloatFromBool(Bool=$Particles.IsTravelingLightning, True Float=0.4, False Float=0.9)
    Index 1 Param 2=0.5
    Index 1 Param 3=1
    Index 2 Param 1=$System.TimeOfDayColorMultiplier
    Index 2 Param 2=4
    Index 2 Param 3=3
    Index 2 Param 4=&MakeCustomFloatFromBool(Bool=$Particles.IsReverseDirection, True Float=1, False Float=0)
    Module.Param0WriteEnabled=true
    Module.Param0WriteEnabled1=true
    Module.Param0WriteEnabled2=true
    Module.Param1WriteEnabled=true
    Module.Param1WriteEnabled1=true
    Module.Param1WriteEnabled2=true
    Module.Param2WriteEnabled=true
    Module.Param2WriteEnabled1=true
    Module.Param2WriteEnabled2=true
    Module.Param3WriteEnabled=true
    Module.Param3WriteEnabled1=false
    Module.Param3WriteEnabled2=true
    Write Parameter Index 1=true
    Write Parameter Index 2=true
   )
   &InitializeParticle(
    Color=<rgba 1 0.4333 0 0.4>
    Lifetime=&Multiply_Float(
     A=&SelectFloatFromArray(Direct Array Index=$Particles.PartitionParticles.Partition, Float Selection Array=$User.LifetimeArray)
     B=&MakeCustomFloatFromBool(Bool=$Particles.IsTravelingLightning, True Float=0.56, False Float=0.433)
    )
    Mesh Scale Mode=::NewEnumerator3
    Mesh Uniform Scale=&MakeCustomFloatFromBool(Bool=&Boolean_OrOperation(Bool1=$Particles.IsHead, Bool2=$Particles.IsLeftSide), True Float=1, False Float=-1)
    UsePositionOffset=false
    Write Material Random=true
   )
   /SetVariables(
    Particles.BackLayerColor:LinearColor=<rgba 1 0 0.7167 -0.2>
    Particles.BackLayerDynamicParameters:Vector4f=&MakeVector4(
     X=&MakeCustomFloatFromBool(Bool=$Particles.IsHead, True Float=1.9, False Float=2.6)
     Y=0.5
     Z=&MakeFloatFromVector4(Vector 4=$Particles.DynamicMaterialParameter, Channel=::NewEnumerator2)
     W=&Multiply_Float(A=&MakeFloatFromVector4(Vector 4=$Particles.DynamicMaterialParameter, Channel=::NewEnumerator3), B=1.6)
    )
    Particles.BitsBackLayerDynamicParameters:Vector4f=&MakeVector4(
     X=&MakeCustomFloatFromBool(Bool=$Particles.IsHead, True Float=1.9, False Float=2.6)
     Y=0.5
     Z=&MakeFloatFromVector4(Vector 4=$Particles.DynamicMaterialParameter, Channel=::NewEnumerator2)
     W=&Multiply_Float(A=&MakeFloatFromVector4(Vector 4=$Particles.DynamicMaterialParameter, Channel=::NewEnumerator3), B=1.6)
    )
   ) {isSetParameters=true}
  ]
  particleUpdateScript [
   &ParticleState()
   &SkeletalMeshLocation(
    Bone / Socket Index=$Particles.BoneSampleIndex
    Bone Sampling Mode=::NewEnumerator4
    MaskSpawnGroup=false
    Skeletal Mesh=$User.SkeletalMesh
    Use Sampled Mesh Attribute as Ribbon ID=false
   )
   &DynamicMaterialParameters(
    Index 0 Param 1=&FloatFromCurve(
     FloatCurve=DataInterface() {propertyValues {Curve {keys [<at 0 1>,<at 0.4 1 RCIM_Constant>,<at 0.45 0 RCIM_Constant>,<at 0.7 1 RCIM_Constant>,<at 1.1 0>]}}}
     Scale Curve=&MakeCustomFloatFromBool(Bool=$Particles.IsHead, True Float=1.66, False Float=2)
    )
    Index 0 Param 4=&FloatFromCurve(
     FloatCurve=DataInterface() {propertyValues {Curve {keys [<at 0 1 RCIM_Constant>,<at 0.4 4 RCIM_Constant>,<at 0.5 1 RCIM_Constant>,<at 1 1 RCIM_Constant>]}}}
     Scale Curve=0.8
    )
    Index 1 Param 1=0.7
    Index 1 Param 2=&FloatFromCurve(FloatCurve=DataInterface() {propertyValues {Curve {keys [{tangentMode=RCTM_User, time=0, value=0}, {tangentMode=RCTM_User, time=1, value=1}]}}})
    Index 1 Param 3=&FloatFromCurve(FloatCurve=DataInterface() {propertyValues {Curve {keys [<at 0 0>,<at 0.35 1 RCIM_Constant>,<at 0.4 4 RCIM_Constant>,<at 0.45 0.25 RCIM_Constant>,<at 0.55 1>,<at 1 0>]}}})
    Index 1 Param 4=&FloatFromCurve(FloatCurve=DataInterface() {propertyValues {Curve {keys [
         {arriveTangent=11.7929, interpMode=RCIM_Cubic, leaveTangent=11.7929, tangentMode=RCTM_User, time=0.25, value=0}
         {interpMode=RCIM_Cubic, tangentMode=RCTM_User, time=0.5, value=1}
         {tangentMode=RCTM_User, time=1, value=0}
        ]}}})
    Module.Param0WriteEnabled=true
    Module.Param0WriteEnabled1=false
    Module.Param1WriteEnabled=false
    Module.Param1WriteEnabled1=true
    Module.Param2WriteEnabled=false
    Module.Param2WriteEnabled1=true
    Module.Param3WriteEnabled=false
    Module.Param3WriteEnabled1=true
    Write Parameter Index 1=true
   ) {module=DynamicMaterialParameters001}
   /SetVariables(Particles.VisibilityTag=&MakeCustomIntFromBool(
     Bool=&SetBoolByFloatComparison(
      A=&DistanceBetweenPositions(
       A=&ReturnCameraPosition(
        Coordinate Space:ENiagaraCoordinateSpace=::Simulation
        Camera Query:NiagaraDataInterfaceCamera=DataInterface() {propertyValues {bRequireCurrentFrameData=false}}
       )
       B=$Particles.Position
      )
      B=&FN_FrontEndFloatModifier(Gameplay=150, Front End=0)
      Comparison Type=::NewEnumerator6
     )
     True INT=1
     False INT=0
    )) {isSetParameters=true}
   /SetVariables(
    Particles.BitsFrontLayerDynamicParameters:Vector4f=&MakeVector4(
     X=&Multiply_Float(
      A=&MakeFloatFromVector4(Vector 4=$Particles.DynamicMaterialParameter, Channel=::NewEnumerator0)
      B=0.75
     )
     Y=&Multiply_Float(A=&MakeFloatFromVector4(Vector 4=$Particles.DynamicMaterialParameter, Channel=::NewEnumerator1), B=1)
     Z=0
     W=0
    )
    Particles.BitsScale:Vector3f=&VectorFromFloat(Value=&FloatFromCurve(
      FloatCurve=DataInterface() {propertyValues {Curve {keys [
          {arriveTangent=4.7868, interpMode=RCIM_Cubic, leaveTangent=4.7868, tangentMode=RCTM_User, time=0, value=0}
          {arriveTangent=-0.423, interpMode=RCIM_Cubic, leaveTangent=-0.423, tangentMode=RCTM_User, time=1, value=0.6666}
         ]}}}
      Scale Curve=&Multiply_Float(
       A=&MakeFloatFromVector(Vector=$Particles.Scale, Channel=::NewEnumerator0)
       B=&RandomRangeFloat(Minimum=0.666, Maximum=1)
      )
     ))
   ) {isSetParameters=true}
   /SetVariables(
    Particles.SpawnBits:NiagaraInt32=&MakeCustomIntFromBool(
     Bool=$Particles.IsHead
     True INT=&MakeCustomIntFromBool(Bool=$User.bisHeadHidden, True INT=1, False INT=$Particles.Initial.SpawnBits)
     False INT=&MakeCustomIntFromBool(
      Bool=$Particles.IsLowerArm
      True INT=&MakeCustomIntFromBool(Bool=$User.bisHandHidden, True INT=1, False INT=$Particles.Initial.SpawnBits)
      False INT=$Particles.Initial.SpawnBits
     )
    )
    Particles.VisibilityTag=&MakeCustomIntFromBool(
     Bool=$Particles.IsHead
     True INT=&MakeCustomIntFromBool(Bool=$User.bisHeadHidden, True INT=1, False INT=0)
     False INT=&MakeCustomIntFromBool(
      Bool=$Particles.IsLowerArm
      True INT=&MakeCustomIntFromBool(Bool=$User.bisHandHidden, True INT=1, False INT=0)
      False INT=0
     )
    )
   ) {isSetParameters=true}
  ]
  rendererClasses [&NiagaraMeshRendererProperties]
  renderers [
   {rendererClass=&NiagaraMeshRendererProperties, rendererIndex=0}
   {rendererClass=&NiagaraMeshRendererProperties, rendererIndex=1}
   {rendererClass=&NiagaraMeshRendererProperties, rendererIndex=2}
   {rendererClass=&NiagaraMeshRendererProperties, rendererIndex=3}
  ]
 }}
summary {
 systemName=NS_Skin_StaticShades_Idle
 emitters [{emitterName=StaticShadesLightning, rendererClasses [&NiagaraMeshRendererProperties]}]
 userVariables [
  User.bIsFrontEnd {type=NiagaraBool}
  User.bIsFrontEndPreview {type=NiagaraBool}
  User.bisHandHidden {type=NiagaraBool}
  User.bisHeadHidden {type=NiagaraBool}
  User.LifetimeArray {
   type=NiagaraDataInterfaceArrayFloat
   defaultValue {
    struct=NiagaraExt_VariableValue_DataInterface
    value {
     dataInterface=&self:NiagaraDataInterfaceArrayFloat_0
     dataInterfaceClass=&NiagaraDataInterfaceArrayFloat
    }
   }
  }
  User.LightningIndex {
   type=NiagaraDataInterfaceArrayInt32
   description=""
   defaultValue {
    struct=NiagaraExt_VariableValue_DataInterface
    value {
     dataInterface=&self:NiagaraDataInterfaceArrayInt32_0
     dataInterfaceClass=&NiagaraDataInterfaceArrayInt32
    }
   }
  }
  User.SideIndex {
   type=NiagaraDataInterfaceArrayInt32
   description=""
   defaultValue {
    struct=NiagaraExt_VariableValue_DataInterface
    value {
     dataInterface=&self:NiagaraDataInterfaceArrayInt32_1
     dataInterfaceClass=&NiagaraDataInterfaceArrayInt32
    }
   }
  }
  User.SkeletalMesh {
   type=NiagaraDataInterfaceSkeletalMesh
   defaultValue {
    struct=NiagaraExt_VariableValue_DataInterface
    value {dataInterface=&self:NiagaraDataInterfaceSkeletalMesh_0}
   }
  }
 ]
}
