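# Material function that adds a glitching, emissive face logo to the StaticShades male character's head, enabled by the StaticSwitchHead static switch.
# Panning gradient sine bands multiply a scrolling glitch-noise mask, gated by time-driven pulses and scaled by GlitchAmount, to offset UVs.
# Those UVs sample the head diffuse as emissive and the FX mask's red channel, which blends it over the incoming material attributes.
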MaterialFunction /Game/Characters/Player/Male/Medium/Bodies/M_MED_StaticShades/Materials/MF_StaticShade_FaceLogo
graph {nodes [
  n1 = FunctionInput() {inputType=FunctionInput_MaterialAttributes}
  n2 = Sine(None=MaterialFunctionCall(UV Channel=Panner() {speedX=3}) {materialFunction=&LinearGradient}.UGradient) {period=10}
  n3 = Panner() {speedX=7}
  n4 = Sine(None=Sine(None=MaterialFunctionCall(UV Channel=@n3) {materialFunction=&LinearGradient}.UGradient) {period=10}) {period=10}
  n5 = ScalarParameter() {defaultValue=0.5, parameterName=Frequency}
  n6 = Add(A=Ceil(None=Sine(None=@n2) {period=10}), B=Ceil(None=Multiply(A=@n4, B=@n5)))
  n7 = TextureCoordinate()
  n8 = TextureSample(UVs=Panner(Coordinate=Multiply(A=@n7) {constB=2}) {speedX=1, speedY=1}) {
   samplerType=SAMPLERTYPE_Masks
   texture=/Game/Characters/Player/Female/Medium/Bodies/F_MED_BlueBadass/Skins/NeonLines/F_MED_BlueBadass_NeonLines_-_GlitchNoise_v1
  }
  n9 = Multiply(A=Time() {period=10}) {constB=0.88}
  n10 = Abs(None=Sine(None=@n9))
  n11 = Multiply(A=@n9) {constB=0.77}
  n12 = Round(None=Multiply(A=OneMinus(None=@n10), B=Frac(None=@n11)))
  n13 = Multiply(A=LinearInterpolate(B=@n8.RGB, Alpha=Saturate(None=@n12))) {constB=0.015}
  n14 = Multiply(A=ComponentMask(None=Multiply(A=@n6, B=@n13)) {g=true, r=true}, B=ScalarParameter() {defaultValue=0.15, parameterName=GlitchAmount})
  n15 = Add(A=TextureCoordinate(), B=@n14)
  n16 = SetMaterialAttributes!(
   MaterialAttributes=GetMaterialAttributes(None=@n1).MaterialAttributes
   EmissiveColor=TextureSample(UVs=@n15) {texture=/Game/Characters/Player/Male/Medium/Bodies/M_MED_StaticShades/Textures/T_M_MED_StaticShades_Head_D}.RGB
  ) {attributeSetTypes [#1]}
  n17 = TextureSample(UVs=@n15) {samplerType=SAMPLERTYPE_Masks, texture=/Game/Characters/Player/Male/Medium/Bodies/M_MED_StaticShades/Textures/T_M_MED_StaticShades_Head_FX}
  FunctionOutput(None=StaticSwitchParameter(True=BlendMaterialAttributes(A=@n1, B=@n16, Alpha=@n17.R), False=@n1) {parameterName=StaticSwitchHead})
  MaterialFunctionCall!(TextureObject, TextureSize, WorldPosition, Export Float 4, In, World Space Normal, ProjectionTransitionContrast) {materialFunction=/Game/Spectating/Effects/WorldAlignedTexture_Glitch}
  OneMinus(None=@n14)
  TextureSample() {texture=&T_M_MED_StaticShades_Head_D}
 ]}
