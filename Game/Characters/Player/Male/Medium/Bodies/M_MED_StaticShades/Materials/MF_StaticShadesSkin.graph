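# Material function for the StaticShades player body that turns masked skin into an animated, glowing surface with black base color, zero specular and full roughness.
# Panning Perlin noise, projected from local position, blends with Fresnel to drive a curve-atlas gradient, with orange layered in from the blue noise channel.
# Time of day scales the emissive.
# A channel of the Alpha input blends it over the incoming material; the panning noise is also output.
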MaterialFunction /Game/Characters/Player/Male/Medium/Bodies/M_MED_StaticShades/Materials/MF_StaticShadesSkin
graph {nodes [
  n1 = GetMaterialAttributes(None=FunctionInput() {inputName=Material, inputType=FunctionInput_MaterialAttributes})
  n2 = MaterialFunctionCall!(
   TextureObject=TextureObject() {samplerType=SAMPLERTYPE_Masks, texture=/Game/Global/Textures/Default/Noise/T_Generic_Perlin}
   TextureSize=ScalarParameter() {defaultValue=100, parameterName=NoiseSize}
   WorldPosition=FunctionInput() {inputName=LocalPos}
   Export Float 4
   In=VectorParameter() {parameterName=PanningAmount, defaultValue=<rgba -0.08 0 0 1>}.RGB
   World Space Normal
   ProjectionTransitionContrast
  ) {materialFunction=/Game/Athena/SafeZone/WorldAlignedTexture_ScalePan}
  n3 = Saturate(None=MaterialFunctionCall(In=@n2.XY Texture, Contrast=ScalarParameter() {defaultValue=0.5, parameterName=Contrast}) {materialFunction=&CheapContrast}.Result)
  n4 = Power(Base=FunctionInput() {inputName=Fresnel, inputType=FunctionInput_Scalar}, Exp=ScalarParameter() {defaultValue=1, parameterName=FresnelBlend})
  n5 = ScalarParameter() {defaultValue=2, parameterName=FresnelBoost}
  n6 = CurveAtlasRowParameter(CurveTime=LinearInterpolate(A=ComponentMask(None=@n3) {g=true}, B=Multiply(A=@n4, B=@n5), Alpha=ScalarParameter() {defaultValue=0.6, parameterName=EdgeBlend})) {
   atlas=&/CA_StaticShades
   curve=&/StaticShadesCurve
  }
  n7 = Saturate(None=@n4)
  n8 = Multiply(A=ComponentMask(None=@n2.XY Texture) {b=true}, B=ScalarParameter() {defaultValue=0.66, parameterName=BoostBlue}) {constB=0.5}
  n9 = ScalarParameter() {defaultValue=3, parameterName=Contrast3}
  n10 = LinearInterpolate(
   A=Multiply(A=@n6, B=LinearInterpolate(Alpha=@n7) {constA=0.3})
   B=VectorParameter() {parameterName=Orange, defaultValue=<rgba 0.3542 0.1071 0 1>}.RGB
   Alpha=Saturate(None=MaterialFunctionCall(In=@n8, Contrast=@n9) {materialFunction=&CheapContrast}.Result)
  )
  n11 = Constant() {r=0.5}
  n12 = MaterialFunctionCall(Value_Morning=@n11, Value_Day=Constant() {r=1}, Value_Evening=@n11, Value_Night=@n11) {materialFunction=&Scalar_Time_of_Day}
  n13 = SetMaterialAttributes!(
   MaterialAttributes=@n1.MaterialAttributes
   EmissiveColor=Multiply(A=Multiply(A=@n10, B=@n12.Result), B=ScalarParameter() {defaultValue=2, parameterName=EmissiveIntensity})
   BaseColor=Constant()
   Roughness=Constant() {r=1}
   Specular=Constant()
  ) {attributeSetTypes [
    #1
    #2
    #3
    #4
   ]}
  n14 = FunctionInput() {inputName=Alpha}
  FunctionOutput(None=BlendMaterialAttributes(A=@n1.MaterialAttributes, B=@n13, Alpha=ChannelMaskParameter(None=@n14) {group=04, parameterName=StaticShadesSkinChannel}.R))
  FunctionOutput(None=@n2.XY Texture) {outputName=PanningPerlin}
 ]}
