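# Material function for the Static Shades player body that adds animated lightning under the skin to Emissive.
# A parallax-offset flipbook and a gradient-based bolt layer share panning UVs warped by stepped noise, with curve atlas rows and time driving flicker and reveal.
# Inverted Fresnel confines the glow to facing areas, and the GMAP input masks it.
# A channel-masked Alpha blends the result into the incoming attributes, behind a static switch.
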MaterialFunction /Game/Characters/Player/Male/Medium/Bodies/M_MED_StaticShades/Materials/MF_UnderskinLightning
graph {nodes [
  n1 = GetMaterialAttributes(None=FunctionInput() {inputType=FunctionInput_MaterialAttributes}) {attributeGetTypes [#1]}
  n2 = Power(Base=FunctionInput() {inputName=GMAP, inputType=FunctionInput_Scalar})
  n3 = Power(Base=OneMinus(None=FunctionInput() {inputName=Fresnel}), Exp=ScalarParameter() {defaultValue=8, parameterName=fff}) {constExponent=1}
  n4 = Time()
  n5 = Frac(None=Multiply(A=@n4, B=ScalarParameter() {defaultValue=1.33, parameterName=AnimationSpeed}))
  n6 = Constant() {r=2}
  n7 = Constant() {r=4}
  n8 = TextureObject() {samplerType=SAMPLERTYPE_LinearColor, texture=/Game/Effects/Fort_Effects/Emotes/EmeraldGlass_Green/T_EmeraldGlass_D_Lightning1}
  n9 = TextureCoordinate()
  n10 = ScalarParameter() {defaultValue=5, parameterName=LightningTiling2}
  n11 = TextureCoordinate() {uTiling=0.46, vTiling=0.2476}
  n12 = Time()
  n13 = Divide(A=Ceil(None=Multiply(A=@n12) {constB=16})) {constB=4}
  n14 = ComponentMask(None=TextureSample(UVs=Panner(Coordinate=@n11, Time=@n13) {speedX=0.4, speedY=0.4}) {samplerType=SAMPLERTYPE_LinearColor, texture=/Game/Global/Textures/Default/Noise/TilingNoise06_N_uncompressed}.RGB) {g=true, r=true}
  n15 = Add(A=Multiply(A=Subtract(A=@n14) {constB=0.5}) {constB=0.1}, B=Panner(Coordinate=TextureCoordinate(), Time=Time()) {speedY=0.2}) {constB=0}
  n16 = BumpOffset(Coordinate=Add(A=Multiply(A=@n9, B=@n10), B=@n15), Height=ScalarParameter() {defaultValue=-2.5, parameterName=LightningBump2})
  n17 = Frac(None=Multiply(A=@n4, B=ScalarParameter() {defaultValue=0.07, parameterName=BiasLightning}))
  n18 = CurveAtlasRowParameter(CurveTime=@n17) {
   atlas=&/CA_StaticShades
   curve=&/LightningCurve
   defaultValue=1
   parameterName=LightningCurves
  }
  n19 = TextureSample(UVs=Multiply(A=ScalarParameter() {defaultValue=12, parameterName=Lightning3Tiling}, B=@n15)) {samplerType=SAMPLERTYPE_LinearColor, texture=/Game/Global/Textures/Default/Masks_and_Gradients/TestLightningGradient}
  n20 = Subtract(A=@n19.G)
  n21 = ScalarParameter() {defaultValue=1.1, parameterName=BiasLightning3}
  n22 = CurveAtlasRowParameter(CurveTime=Frac(None=Multiply(A=@n4, B=@n21))) {
   atlas=&CA_StaticShades
   curve=&LightningCurve
   defaultValue=1
   parameterName=LightningCurves
  }
  n23 = LinearInterpolate(Alpha=@n22.B) {constA=-0.1, constB=1.25}
  n24 = Multiply(A=@n19.R, B=Round(None=Add(A=@n20, B=@n23)))
  n25 = SphereMask(A=LinearInterpolate(Alpha=@n17) {constA=-0.33, constB=1.25}, B=@n19.B) {attenuationRadius=0.2}
  n26 = Add(
   A=Multiply(
    A=MaterialFunctionCall("Animation  Phase (0-1)"=@n5, Number of Rows=@n6, Number of Columns=@n7, Texture=@n8, UVs=@n16) {materialFunction=&FlipBook}.Result
    B=@n18.G
   )
   B=Multiply(A=Multiply(A=@n24, B=@n25)) {constB=0.75}
  )
  n27 = Multiply(A=Saturate(None=Multiply(A=@n3, B=@n26)), B=ScalarParameter() {defaultValue=0.5, parameterName=LightningOpacity})
  n28 = SetMaterialAttributes!(
   MaterialAttributes=@n1.MaterialAttributes
   EmissiveColor=LinearInterpolate(
    A=@n1.EmissiveColor
    B=VectorParameter() {parameterName=Lightning Color, defaultValue=<rgba 10 8.8934 2.2466 1>}.RGB
    Alpha=Multiply(A=@n2, B=@n27)
   )
  ) {attributeSetTypes [#1]}
  n29 = ChannelMaskParameter(None=FunctionInput() {inputName=Alpha}) {parameterName=LightningChannel}
  FunctionOutput(None=StaticSwitchParameter(True=BlendMaterialAttributes(A=@n1.MaterialAttributes, B=@n28, Alpha=@n29.R), False=@n1.MaterialAttributes) {defaultValue=true, parameterName=UseUnderskinLightning})
 ]}
