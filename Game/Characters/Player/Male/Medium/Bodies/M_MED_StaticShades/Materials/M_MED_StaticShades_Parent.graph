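# Parent material for the Static Shades player character body (male, medium), built as a Material Attributes chain of shared functions.
# The 2023 character PreFX base feeds a Static Shades skin function, then underskin lightning, a face logo layer and the character PostFX function.
# Panning Perlin noise from the skin, multiplied by the Tech Art mask, drives lightning alpha, alongside Fresnel and the M map's green channel.
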Material /Game/Characters/Player/Male/Medium/Bodies/M_MED_StaticShades/Materials/M_MED_StaticShades_Parent
graph {
 properties {
  bUseMaterialAttributes=true
  bUsedWithMeshDeformer=true
  bUsedWithMorphTargets=true
  bUsedWithSkeletalMesh=true
  pixelDepthOffsetMode=PDOM_Legacy
  usedShadingModels=Default Lit
 }
 nodes [
  n1 = MaterialFunctionCall() {materialFunction=&MF_CharacterParent_2023_PreFX}
  n2 = MaterialFunctionCall!(
   LocalPos=@n1.Pre skinned Local Position
   Alpha=@n1.Pre skinned Local Position
   Fresnel=@n1.Pre skinned Local Position
   Material=@n1.Pre skinned Local Position
  ) {materialFunction=&/MF_StaticShadesSkin}
  n3 = Multiply(A=ComponentMask(None=@n2.PanningPerlin) {r=true}, B=@n1.Tech Art Mask)
  n4 = MaterialFunctionCall(Float3=@n1.M) {materialFunction=&BreakOutFloat3Components}
  MaterialFunctionCall(
   Material=MaterialFunctionCall!(In=MaterialFunctionCall!(Alpha=@n3, Fresnel=@n1.Fresnel, GMAP=@n4.G, In=@n2.Result) {materialFunction=&/MF_UnderskinLightning}.Result) {materialFunction=&/MF_StaticShade_FaceLogo}.Result
   M=@n1.M
   Pre Skinned Local Position=@n1.M
   Distance Blend=@n1.M
   Fresnel=@n1.M
  ) {materialFunction=&MF_CharacterParent_2023_PostFX}
 ]
}
