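# Physics asset for the StaticShades medium male player body, set up for physics on its dangling clothing: shirt hems, belt strap, front straps, zipper strap and right-side straps.
# Spheres and capsules cover those dyn_ bones, while larger capsules cover pelvis, spine and thighs.
# Angle-limited constraints hang hem and belt chains from the pelvis and straps from spine_04.
# Thigh and lower spine joints are free in swing and twist.
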PhysicsAsset /Game/Characters/Player/Male/Medium/Bodies/M_MED_StaticShades/Meshes/M_MED_StaticShades_Physics
bodies {
 dyn_belt_strap_1 [{shapeType=Sphere, radius=1.5, shapeName=dyn_belt_strap_1_sphere, center=<2.2903 1.263 1.3224>}]
 dyn_belt_strap_2 [{shapeType=Sphere, radius=1.5, shapeName=dyn_belt_strap_2_sphere, center=<3.0191 -0.1811 0.216>}]
 dyn_front_strap_1 [{shapeType=Sphere, radius=3, shapeName=dyn_front_strap_1_sphere, center=<4.6149 1.1152 0>}]
 dyn_front_strap_2 [{shapeType=Sphere, radius=3.5, shapeName=dyn_front_strap_2_sphere, center=<1.6989 1.6627 0>}]
 dyn_front_strap_3 [{shapeType=Sphere, radius=3, shapeName=dyn_front_strap_3_sphere, center=<0.9676 1.7469 0>}]
 dyn_front_strap_zipper_1 [{shapeType=Sphere, radius=3.25, shapeName=dyn_front_strap_zipper_1_sphere, center=<1.4891 -0.2123 0>}]
 dyn_lower_right_strap_2 [{shapeType=Sphere, radius=2, shapeName=dyn_lower_right_strap_2_sphere, center=<1.8378 0 0.1817>}]
 dyn_shirt_hem_fr_l_1 [
  {shapeType=Sphere, radius=2.75, shapeName=dyn_shirt_hem_fr_l_1_sphere, center=<4.0351 -6.3771 -0.6988>}
  {
   shapeType=Capsule
   length=4
   radius=3
   shapeName=dyn_shirt_hem_fr_l_1_capsule
   center=<4.3862 4.3936 1.9403>
   rotation=<rot 0.0004 0.0003 110>
  }
 ]
 dyn_shirt_hem_fr_l_2 [{
   shapeType=Capsule
   length=4.5
   radius=2.5
   shapeName=dyn_shirt_hem_fr_l_2_capsule
   center=<4.0842 3.8063 1.6764>
   rotation=<rot 0.0003 0.0002 115>
  }]
 dyn_shirt_hem_fr_r_1 [{shapeType=Capsule, length=6, radius=3, shapeName=dyn_shirt_hem_fr_r_1_capsule, center=<1.0137 0.3627 0.1664>, rotation=<rot 0 0 90>}]
 dyn_shirt_hem_fr_r_2 [{shapeType=Sphere, radius=3.7701, shapeName=dyn_shirt_hem_fr_r_2_sphere, center=<2.6819 0.4399 3.1366>}]
 dyn_shirt_hem_l_1 [{shapeType=Sphere, radius=3, shapeName=dyn_shirt_hem_l_1_sphere, center=<2.4381 0 2.5768>}]
 dyn_shirt_hem_r_1 [
  {shapeType=Sphere, radius=3.5, shapeName=dyn_shirt_hem_r_1_sphere, center=<5.0382 -9.5529 0.1863>}
  {shapeType=Capsule, length=12.8949, radius=3.9674, center=<3.5302 0.7825 1.4348>, rotation=<rot -2.5767 9.6655 -70.2178>}
 ]
 dyn_shirt_hem_r_2 [{shapeType=Sphere, radius=3, shapeName=dyn_shirt_hem_r_2_sphere, center=<4.3231 0 1.8536>}]
 dyn_upper_right_strap_1 [{shapeType=Sphere, radius=2, shapeName=dyn_upper_right_strap_1_sphere, center=<1.8378 0 0.0732>}]
 pelvis [{shapeType=Capsule, length=12.5, radius=9.25, center=<0 0 0.5243>}]
 spine_01 [{shapeType=Capsule, length=7.7903, radius=9.5, shapeName=spine_01_capsule}]
 spine_03 [{shapeType=Capsule, length=2.4054, radius=11.5443, shapeName=spine_03_capsule, center=<0 -0.5451 0.5002>}]
 spine_04 [
  {shapeType=Sphere, radius=5, shapeName=spine_04_sphere, center=<-0.3615 14.1231 -6.6079>}
  {shapeType=Capsule, length=8.5, radius=9, center=<0 1.8736 0>}
 ]
 spine_05 [{shapeType=Capsule, length=10, radius=10, center=<0 1.0604 0>}]
 thigh_l [{shapeType=Capsule, length=28, radius=8, shapeName=thigh_l_capsule, center=<-8.9882 0.7798 0>, rotation=<rot 90 180 -174.9999>}]
 thigh_r [{shapeType=Capsule, length=28, radius=8, shapeName=thigh_r_capsule, center=<9 0 0>, rotation=<rot 90 0 0>}]
}
constraints [
 {bone1Name=spine_04, bone2Name=spine_03, swing1LimitDegrees=45, swing2LimitDegrees=45, twistLimitDegrees=45}
 {bone1Name=spine_05, bone2Name=spine_04, swing1LimitDegrees=45, swing2LimitDegrees=45, twistLimitDegrees=45}
 {bone1Name=dyn_shirt_hem_r_1, bone2Name=pelvis, swing1LimitDegrees=20, swing2LimitDegrees=120, twistLimitDegrees=10}
 {bone1Name=dyn_shirt_hem_r_2, bone2Name=dyn_shirt_hem_r_1, swing1LimitDegrees=20, swing2LimitDegrees=30, twistLimitDegrees=10}
 {bone1Name=dyn_shirt_hem_l_1, bone2Name=pelvis, swing1LimitDegrees=20, swing2LimitDegrees=30, twistLimitDegrees=10}
 {bone1Name=dyn_shirt_hem_fr_r_1, bone2Name=pelvis, swing1LimitDegrees=0, swing2LimitDegrees=30, twistLimitDegrees=10}
 {bone1Name=dyn_shirt_hem_fr_r_2, bone2Name=dyn_shirt_hem_fr_r_1, swing1LimitDegrees=5, swing2LimitDegrees=60, twistLimitDegrees=10}
 {bone1Name=dyn_shirt_hem_fr_l_2, bone2Name=dyn_shirt_hem_fr_l_1, swing1LimitDegrees=10, swing2LimitDegrees=30, twistLimitDegrees=10}
 {bone1Name=dyn_shirt_hem_fr_l_1, bone2Name=pelvis, swing1LimitDegrees=10, swing2LimitDegrees=30, twistLimitDegrees=10}
 {bone1Name=dyn_belt_strap_2, bone2Name=dyn_belt_strap_1, swing1LimitDegrees=25, swing2LimitDegrees=20, twistLimitDegrees=5}
 {bone1Name=dyn_belt_strap_1, bone2Name=pelvis, swing1LimitDegrees=25, swing2LimitDegrees=20, twistLimitDegrees=5}
 {
  bone1Name=spine_03
  bone2Name=spine_01
  swing1LimitDegrees=45
  swing1Motion=Free
  swing2LimitDegrees=45
  swing2Motion=Free
  twistLimitDegrees=45
  twistMotion=Free
 }
 {bone1Name=dyn_front_strap_1, bone2Name=spine_04, swing1LimitDegrees=30, swing2LimitDegrees=30, twistLimitDegrees=10}
 {bone1Name=dyn_front_strap_2, bone2Name=dyn_front_strap_1, swing1LimitDegrees=30, swing2LimitDegrees=30, twistLimitDegrees=10}
 {bone1Name=dyn_front_strap_zipper_1, bone2Name=spine_04, swing1LimitDegrees=20, swing2LimitDegrees=10, twistLimitDegrees=10}
 {bone1Name=dyn_front_strap_3, bone2Name=dyn_front_strap_2, swing1LimitDegrees=30, swing2LimitDegrees=30, twistLimitDegrees=10}
 {bone1Name=dyn_upper_right_strap_1, bone2Name=spine_04, swing1LimitDegrees=15, swing2LimitDegrees=30, twistLimitDegrees=10}
 {bone1Name=dyn_lower_right_strap_2, bone2Name=spine_04, swing1LimitDegrees=15, swing2LimitDegrees=30, twistLimitDegrees=10}
 {
  bone1Name=thigh_r
  bone2Name=pelvis
  swing1LimitDegrees=45
  swing1Motion=Free
  swing2LimitDegrees=45
  swing2Motion=Free
  twistLimitDegrees=45
  twistMotion=Free
 }
 {
  bone1Name=thigh_l
  bone2Name=pelvis
  swing1LimitDegrees=45
  swing1Motion=Free
  swing2LimitDegrees=45
  swing2Motion=Free
  twistLimitDegrees=45
  twistMotion=Free
 }
 {
  bone1Name=spine_01
  bone2Name=pelvis
  swing1LimitDegrees=45
  swing1Motion=Free
  swing2LimitDegrees=45
  swing2Motion=Free
  twistLimitDegrees=45
  twistMotion=Free
 }
]
