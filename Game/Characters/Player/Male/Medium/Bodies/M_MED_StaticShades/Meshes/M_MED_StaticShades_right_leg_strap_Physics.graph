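# Physics asset for the right leg strap mesh on the medium male player StaticShades body, defining collision and joints for a dangling strap.
# Three sphere bodies form a strap chain, with the first constrained to the right calf and each following one to the segment before it, all with limited swing and tighter twist.
# Capsules on the right calf and thigh give the leg collision volumes alongside the chain.
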PhysicsAsset /Game/Characters/Player/Male/Medium/Bodies/M_MED_StaticShades/Meshes/M_MED_StaticShades_right_leg_strap_Physics
bodies {
 calf_r [{shapeType=Capsule, length=34, radius=6, shapeName=calf_r_capsule, center=<19.2115 0.2678 0>, rotation=<rot -90 179.9999 179.9989>}]
 right_leg_strap_1 [{shapeType=Sphere, radius=2.75, shapeName=right_leg_strap_1_sphere, center=<3.9001 -0.7913 0>}]
 right_leg_strap_2 [{shapeType=Sphere, radius=3, shapeName=right_leg_strap_2_sphere, center=<3.7701 -0.2998 0>}]
 right_leg_strap_3 [{shapeType=Sphere, radius=3, shapeName=right_leg_strap_3_sphere, center=<2.9723 -0.1964 0>}]
 thigh_r [{shapeType=Capsule, length=28, radius=8, shapeName=thigh_r_capsule, center=<9 0 0>, rotation=<rot 90 0 0>}]
}
constraints [
 {bone1Name=right_leg_strap_1, bone2Name=calf_r, swing1LimitDegrees=30, swing2LimitDegrees=30, twistLimitDegrees=10}
 {bone1Name=right_leg_strap_2, bone2Name=right_leg_strap_1, swing1LimitDegrees=30, swing2LimitDegrees=30, twistLimitDegrees=10}
 {bone1Name=right_leg_strap_3, bone2Name=right_leg_strap_2, swing1LimitDegrees=30, swing2LimitDegrees=30, twistLimitDegrees=10}
]
