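# Post-process AnimBP for the Male Medium StaticShades body, animating straps, grenades, carabiner, zipper, buckle and shirt hem over the attached parent's copied pose.
# Nested bool blends select RigidBody with corrective bone transforms, an AnimDynamics chain, or a simple Trail fallback; the first two start with a deformer Control Rig.
# The EventGraph derives thigh-bend, movement, standing and skydiving alphas, updates gravity override, and drives material WPO and skydive wind parameters.
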AnimBlueprint /Game/Characters/Player/Male/Medium/Bodies/M_MED_StaticShades/Meshes/M_MED_StaticShades_AnimBP
target_skeleton=&/M_MED_StaticShades_Skeleton
graphs [
 AnimGraph {nodes [
   n1 = Trail controller - Bone: right_leg_strap_3(ComponentPose=Local To Component(LocalPose=Copy Pose From Mesh() {bUseAttachedParent=true}), Alpha=1) {
    bLimitRotation=true
    bShowBaseMotion=false
    bUsePlanarLimit=true
    chainLength=3
    planarLimits [
     {drivingBone {boneName=calf_r}, planeTransform {location=<0 5 0>, rotation=<rot 0 0 90>, scale=<1 1 1>}}
     {drivingBone {boneName=thigh_r}, planeTransform {location=<0 5 0>, rotation=<rot 0 0 90>, scale=<1 1 1>}}
    ]
    rotationLimits [{limitMax=<90 90 90>, limitMin=<-90 -90 -90>}, {limitMax=<180 180 180>}, {limitMax=<180 180 180>}]
    rotationOffsets [<0 0 0>, <0 0 0>, <0 0 0>]
    trailBone {boneName=right_leg_strap_3}
    trailRelaxationSpeed {editorCurveData {keys [<at 1 3 RCIM_Linear>,<at 2 1 RCIM_Linear>]}}
   }
   n2 = Get Not skydiving() -> (Not skydiving:double)
   n3 = Get Skydiving() -> (Skydiving:double)
   n4 = Get Gravity Override()
   n5 = Control Rig(Source=Copy Pose From Mesh() {bUseAttachedParent=true}) {
    alphaBoolBlend {blendOption=Linear}
    controlRigAssetReference {blueprintRigClass=&/M_MED_StaticShades_CtrlRig_deformers.M_MED_StaticShades_CtrlRig_deformers_C}
   }
   n6 = "Anim Dynamics - Chain: right_leg_strap_1 -> right_leg_strap_3"!(ComponentPose=Local To Component(LocalPose=@n5), Alpha=0.5, GravityOverride:Vector="0,0,-980", ExternalForce="0,0,0") -> (Pose) {
    angularDampingOverride=1
    angularSpringConstant=3000
    bAngularSpring=true
    bChain=true
    bOverrideAngularDamping=true
    bUseGravityOverride=true
    bUsePlanarLimit=false
    bUseSphericalLimits=true
    simulationSpace=RootRelative
    alphaBoolBlend {blendOption=Linear}
    boundBone {boneName=right_leg_strap_1}
    chainEnd {boneName=right_leg_strap_3}
    componentAppliedLinearAccClamp=<2000 2000 2000>
    componentLinearAccScale=<0.01 0.01 0.01>
    componentLinearVelScale=<0.01 0.01 0.01>
    physicsBodyDefinitions [
     {
      boundBone {boneName=right_leg_strap_1}
      boxExtents=<3 5 5>
      constraintSetup {angularLimitsMax=<10 20 30>, angularLimitsMin=<-10 -20 -30>, angularTarget=<0 0 0>}
      localJointOffset=<3 0 0>
     }
     {
      boundBone {boneName=right_leg_strap_2}
      boxExtents=<3 5 5>
      constraintSetup {angularLimitsMax=<10 20 30>, angularLimitsMin=<-10 -20 -30>, angularTarget=<0 0 0>}
      localJointOffset=<3 0 0>
     }
     {
      boundBone {boneName=right_leg_strap_3}
      boxExtents=<3 5 5>
      constraintSetup {angularLimitsMax=<10 20 30>, angularLimitsMin=<-10 -20 -30>, angularTarget=<0 0 0>}
      localJointOffset=<3 0 0>
     }
    ]
    sphericalLimits [
     {limitRadius=7, drivingBone {boneName=calf_r}, sphereLocalOffset=<10 2 0>}
     {limitRadius=7, drivingBone {boneName=calf_r}, sphereLocalOffset=<22 2 0>}
    ]
   }
   n7 = Anim Dynamics - Bone: dyn_grenade_1_2!(ComponentPose=@n6, Alpha=0.8, GravityOverride:Vector="0,0,-980", ExternalForce="0,0,0") -> (Pose) {
    angularDampingOverride=0.7
    angularSpringConstant=100
    bAngularSpring=true
    bOverrideAngularDamping=true
    bUseGravityOverride=true
    linearSpringConstant=100
    alphaBoolBlend {blendOption=Linear}
    boundBone {boneName=dyn_grenade_1_2}
    componentAppliedLinearAccClamp=<2000 2000 2000>
    componentLinearAccScale=<0.5 0.5 0.5>
    componentLinearVelScale=<0.5 0.5 0.5>
    gravityOverride {z=-50}
    physicsBodyDefinitions [{
      boundBone {boneName=dyn_grenade_1_2}
      boxExtents=<8 8 3>
      constraintSetup {angularLimitsMax=<10 0 10>, angularLimitsMin=<-10 -130 -10>, angularTarget=<1 0 0>}
      localJointOffset=<0.5 0 -1.5>
     }]
    relativeSpaceBone {boneName=dyn_grenade_1_1}
   }
   n8 = Anim Dynamics - Bone: dyn_carabiner_1!(ComponentPose=@n7, Alpha=1, GravityOverride:Vector="0,0,-1960", ExternalForce="0,0,0") -> (Pose) {
    angularDampingOverride=1
    angularSpringConstant=10000
    bAngularSpring=true
    bOverrideAngularDamping=true
    bUseGravityOverride=true
    alphaBoolBlend {blendOption=Linear}
    boundBone {boneName=dyn_carabiner_1}
    componentAppliedLinearAccClamp=<2000 2000 2000>
    componentLinearAccScale=<0.6 0.6 0.6>
    componentLinearVelScale=<0.6 0.6 0.6>
    physicsBodyDefinitions [{
      boundBone {boneName=dyn_carabiner_1}
      boxExtents=<20 40 40>
      constraintSetup {angularLimitsMax=<30 10 30>, angularLimitsMin=<-30 -140 -30>, angularTarget=<0 0 0>}
      localJointOffset=<-20 0 0>
     }]
   }
   n9 = "Anim Dynamics - Chain: dyn_front_strap_1 -> dyn_front_strap_3"!(ComponentPose=@n8, Alpha=0.5, GravityOverride:Vector="0,0,-980", ExternalForce="0,0,0") -> (Pose) {
    angularDampingOverride=1
    angularSpringConstant=5000
    bAngularSpring=true
    bChain=true
    bOverrideAngularDamping=true
    bUseGravityOverride=true
    bUsePlanarLimit=false
    bUseSphericalLimits=true
    simulationSpace=RootRelative
    alphaBoolBlend {blendOption=Linear}
    boundBone {boneName=dyn_front_strap_1}
    chainEnd {boneName=dyn_front_strap_3}
    componentAppliedLinearAccClamp=<2000 2000 2000>
    componentLinearAccScale=<0.01 0.01 0.01>
    componentLinearVelScale=<0.01 0.01 0.01>
    physicsBodyDefinitions [
     {
      boundBone {boneName=dyn_front_strap_1}
      boxExtents=<2 6 6>
      constraintSetup {angularLimitsMax=<5 15 55>, angularLimitsMin=<-5 -15 -5>, angularTarget=<0 0 0>}
      localJointOffset=<6 0 -2>
     }
     {
      boundBone {boneName=dyn_front_strap_2}
      boxExtents=<2 6 6>
      constraintSetup {angularLimitsMax=<5 15 35>, angularLimitsMin=<-5 -15 -5>, angularTarget=<0 0 0>}
      localJointOffset=<6 0 -3>
     }
     {
      boundBone {boneName=dyn_front_strap_3}
      boxExtents=<2 6 6>
      constraintSetup {angularLimitsMax=<5 15 25>, angularLimitsMin=<-5 -15 -5>, angularTarget=<0 0 0>}
      localJointOffset=<6 0 -1>
     }
    ]
    relativeSpaceBone {boneName=pelvis}
    sphericalLimits [
     {limitRadius=14.5, drivingBone {boneName=spine_04}, sphereLocalOffset=<0 0 0>}
     {limitRadius=12, drivingBone {boneName=spine_03}, sphereLocalOffset=<0 0 0>}
     {limitRadius=14, drivingBone {boneName=spine_05}, sphereLocalOffset=<0 0 0>}
    ]
   }
   n10 = Anim Dynamics - Bone: dyn_grenade_2_2!(ComponentPose=@n9, Alpha=1, GravityOverride:Vector="0,0,-980", ExternalForce="0,0,0") -> (Pose) {
    angularDampingOverride=0.7
    angularSpringConstant=100
    bAngularSpring=true
    bOverrideAngularDamping=true
    bUseGravityOverride=true
    linearSpringConstant=100
    alphaBoolBlend {blendOption=Linear}
    boundBone {boneName=dyn_grenade_2_2}
    componentAppliedLinearAccClamp=<2000 2000 2000>
    componentLinearAccScale=<0.6 0.6 0.6>
    componentLinearVelScale=<0.6 0.6 0.6>
    gravityOverride {z=-50}
    physicsBodyDefinitions [{
      boundBone {boneName=dyn_grenade_2_2}
      boxExtents=<8 8 3>
      constraintSetup {angularLimitsMax=<15 0 10>, angularLimitsMin=<-15 -130 -10>, angularTarget=<1 0 0>}
      localJointOffset=<0.5 0 -1.5>
     }]
    relativeSpaceBone {boneName=dyn_grenade_2_1}
   }
   n11 = Anim Dynamics - Bone: left_leg_zipper_1!(ComponentPose=@n10, Alpha=1, GravityOverride:Vector="0,0,-300", ExternalForce="0,0,0") -> (Pose) {
    angularDampingOverride=0.999
    angularSpringConstant=1000
    bAngularSpring=true
    bUseGravityOverride=true
    alphaBoolBlend {blendOption=Linear}
    boundBone {boneName=left_leg_zipper_1}
    componentAppliedLinearAccClamp=<1000 1000 1000>
    componentLinearAccScale=<0.2 0.2 0.2>
    componentLinearVelScale=<0.2 0.2 0.2>
    physicsBodyDefinitions [{
      boundBone {boneName=left_leg_zipper_1}
      boxExtents=<10 10 10>
      constraintSetup {angularLimitsMax=<0 0 5>, angularLimitsMin=<0 -90 -5>, angularTarget=<1 0 0>}
      localJointOffset=<2 0 1>
     }]
   }
   n12 = Anim Dynamics - Bone: dyn_grenade_1_1!(ComponentPose=@n11, Alpha=@n3, GravityOverride:Vector=@n4, ExternalForce="0,0,0") -> (Pose) {
    angularBiasOverride=0.999
    angularDampingOverride=0.999
    angularSpringConstant=1000
    bAngularSpring=true
    bOverrideAngularDamping=true
    bUseGravityOverride=true
    simulationSpace=BoneRelative
    alphaBoolBlend {blendOption=Linear}
    boundBone {boneName=dyn_grenade_1_1}
    componentAppliedLinearAccClamp=<1000 1000 1000>
    componentLinearAccScale=<0.2 0.2 0.2>
    componentLinearVelScale=<0.2 0.2 0.2>
    physicsBodyDefinitions [{
      collisionType=CustomSphere
      sphereCollisionRadius=1
      boundBone {boneName=dyn_grenade_1_1}
      boxExtents=<40 60 60>
      constraintSetup {angularLimitsMax=<0 20 25>, angularLimitsMin=<0 -20 0>, angularTarget=<1 0 0>}
      localJointOffset=<-2 0 0>
     }]
    relativeSpaceBone {boneName=root}
   }
   n13 = Anim Dynamics - Bone: dyn_grenade_2_1!(ComponentPose=@n12, Alpha=@n3, GravityOverride:Vector=@n4, ExternalForce="0,0,0") -> (Pose) {
    angularBiasOverride=0.999
    angularDampingOverride=0.999
    angularSpringConstant=1000
    bAngularSpring=true
    bOverrideAngularDamping=true
    bUseGravityOverride=true
    linearSpringConstant=100
    simulationSpace=BoneRelative
    alphaBoolBlend {blendOption=Linear}
    boundBone {boneName=dyn_grenade_2_1}
    componentAppliedLinearAccClamp=<1000 1000 1000>
    componentLinearAccScale=<0.2 0.2 0.2>
    componentLinearVelScale=<0.2 0.2 0.2>
    physicsBodyDefinitions [{
      collisionType=CustomSphere
      sphereCollisionRadius=1
      boundBone {boneName=dyn_grenade_2_1}
      boxExtents=<30 50 50>
      constraintSetup {angularLimitsMax=<0 20 7>, angularLimitsMin=<0 -20 0>, angularTarget=<1 0 0>}
      localJointOffset=<-1.8 0 0>
     }]
    relativeSpaceBone {boneName=root}
   }
   n14 = Anim Dynamics - Bone: dyn_grenade_1_1(ComponentPose=@n13, Alpha=@n2, GravityOverride="0,0,-1500", ExternalForce="0,0,0") {
    angularDampingOverride=0.999
    angularSpringConstant=1000
    bAngularSpring=true
    bOverrideAngularDamping=true
    bUseGravityOverride=true
    simulationSpace=BoneRelative
    alphaBoolBlend {blendOption=Linear}
    boundBone {boneName=dyn_grenade_1_1}
    componentAppliedLinearAccClamp=<2000 2000 2000>
    componentLinearAccScale=<0.2 0.2 0.2>
    componentLinearVelScale=<0.2 0.2 0.2>
    physicsBodyDefinitions [{
      boundBone {boneName=dyn_grenade_1_1}
      boxExtents=<20 27 27>
      constraintSetup {angularLimitsMax=<20 20 5>, angularLimitsMin=<-20 -20 -5>, angularTarget=<1 0 0>}
      localJointOffset=<-2 0 0>
     }]
    relativeSpaceBone {boneName=root}
   }
   n15 = Anim Dynamics - Bone: dyn_grenade_2_1(ComponentPose=@n14, Alpha=@n2, GravityOverride="0,0,-1500", ExternalForce="0,0,0") {
    angularDampingOverride=0.999
    angularSpringConstant=1000
    bAngularSpring=true
    bOverrideAngularDamping=true
    bUseGravityOverride=true
    linearSpringConstant=100
    simulationSpace=BoneRelative
    alphaBoolBlend {blendOption=Linear}
    boundBone {boneName=dyn_grenade_2_1}
    componentAppliedLinearAccClamp=<2000 2000 2000>
    componentLinearAccScale=<0.2 0.2 0.2>
    componentLinearVelScale=<0.2 0.2 0.2>
    physicsBodyDefinitions [{
      boundBone {boneName=dyn_grenade_2_1}
      boxExtents=<20 30 30>
      constraintSetup {angularLimitsMax=<20 20 5>, angularLimitsMin=<-20 -20 -5>, angularTarget=<1 0 0>}
      localJointOffset=<-1.8 0 0>
     }]
    relativeSpaceBone {boneName=root}
   }
   n16 = "Constraint - dyn_front_strap_zipper_MSTR (PPPP)"!(
    ComponentPose=@n15
    Alpha=1
    ConstraintWeights_0=0.023
    ConstraintWeights_1:float=0.122
    ConstraintWeights_2:float=0.475
    ConstraintWeights_3:float=0.38
   ) -> (Pose) {
    alphaBoolBlend {blendOption=Linear}
    boneToModify {boneName=dyn_front_strap_zipper_MSTR}
    constraintSetup [{targetBone {boneName=spine_02}}, {targetBone {boneName=spine_03}}, {targetBone {boneName=spine_04}}, {targetBone {boneName=spine_05}}]
    constraintWeights [1,1,1,1]
   }
   n17 = "Constraint - dyn_carabiner_MSTR (PPPPPP)"!(
    ComponentPose=@n16
    Alpha=1
    ConstraintWeights_0=0.139
    ConstraintWeights_1:float=0.332
    ConstraintWeights_2:float=0.367
    ConstraintWeights_3:float=0.126
    ConstraintWeights_4:float=0.018
    ConstraintWeights_5:float=0.018
   ) -> (Pose) {
    alphaBoolBlend {blendOption=Linear}
    boneToModify {boneName=dyn_carabiner_MSTR}
    constraintSetup [
     {targetBone {boneName=spine_02}}
     {targetBone {boneName=spine_03}}
     {targetBone {boneName=spine_04}}
     {targetBone {boneName=spine_05}}
     {targetBone {boneName=spine_01}}
     {targetBone {boneName=clavicle_l}}
    ]
    constraintWeights [1,1,1,1,1,1]
   }
   n18 = Blend Poses by bool(
    BlendPose_0=Component To Local(ComponentPose=@n17)
    BlendPose_1=Component To Local(ComponentPose=@n1)
    bActiveValue=Get bAnimDynamicsEnabled() -> (bAnimDynamicsEnabled:bool)
   ) {blendType=Linear}
   n19 = Get Not skydiving()
   n20 = Get Skydiving()
   n21 = Get Gravity Override()
   n22 = Get RightThighBendAlpha() -> (RightThighBendAlpha:double)
   n23 = Control Rig(Source=Copy Pose From Mesh() {bUseAttachedParent=true}) {
    alphaBoolBlend {blendOption=Linear}
    controlRigAssetReference {blueprintRigClass=&M_MED_StaticShades_CtrlRig_deformers_C}
   }
   n24 = Get Gravity Override()
   n25 = RigidBody!(ComponentPose=Local To Component(LocalPose=@n23), Alpha, OverrideWorldGravity:Vector=@n24, ExternalForce) -> (Pose) {
    bClampLinearTranslationLimitToRefPose=true
    bOverrideWorldGravity=true
    overridePhysicsAsset=&/M_MED_StaticShades_Physics
    simulationSpace=BaseBoneSpace
    alphaBoolBlend {blendOption=Linear}
    baseBoneRef {boneName=spine_01}
    componentAppliedLinearAccClamp=<2000 2000 2000>
    componentLinearAccScale=<0.4 0.4 0.4>
    componentLinearVelScale=<0.4 0.4 0.4>
    overrideWorldGravity {z=-980}
   }
   n26 = RigidBody(ComponentPose=@n25, OverrideWorldGravity=@n24) {
    bClampLinearTranslationLimitToRefPose=true
    bOverrideWorldGravity=true
    overridePhysicsAsset=&/M_MED_StaticShades_right_leg_strap_Physics
    simulationSpace=BaseBoneSpace
    alphaBoolBlend {blendOption=Linear}
    baseBoneRef {boneName=calf_r}
    componentAppliedLinearAccClamp=<2000 2000 2000>
    componentLinearAccScale=<0.6 0.6 0.6>
    componentLinearVelScale=<0.6 0.6 0.6>
    overrideWorldGravity {z=-980}
   }
   n27 = Get LeftThighBendAlpha() -> (LeftThighBendAlpha:double)
   n28 = Look At - Bone: dyn_shirt_hem_fr_l_MSTR(ComponentPose=@n26, Alpha=@n27) {
    alphaBoolBlend {blendOption=Linear}
    boneToModify {boneName=dyn_shirt_hem_fr_l_MSTR}
    lookAtLocation {x=-20, z=-5}
    lookAtTarget {boneReference {boneName=calf_l}}
    lookAt_Axis {axis=<1 0>}
   }
   n29 = "Transform (Modify) Bone - Bone: dyn_shirt_hem_fr_l_MSTR"(ComponentPose=@n28, Alpha=@n27, Translation="0,0,0", Rotation="20,0,-20", Scale="1,1,1") {
    rotationMode=BMM_Additive
    rotationSpace=BCS_BoneSpace
    translationMode=BMM_Additive
    translationSpace=BCS_BoneSpace
    alphaBoolBlend {blendOption=Linear}
    boneToModify {boneName=dyn_shirt_hem_fr_l_MSTR}
    rotation {pitch=20, roll=-20}
   }
   n30 = "Transform (Modify) Bone - Bone: dyn_shirt_hem_fr_l_2"(ComponentPose=@n29, Alpha=@n27, Translation="0,0,0", Rotation="-30,0,0", Scale="1,1,1") {
    rotationMode=BMM_Additive
    rotationSpace=BCS_BoneSpace
    translationMode=BMM_Additive
    translationSpace=BCS_BoneSpace
    alphaBoolBlend {blendOption=Linear}
    boneToModify {boneName=dyn_shirt_hem_fr_l_2}
    rotation {pitch=-30}
   }
   n31 = "Transform (Modify) Bone - Bone: dyn_shirt_hem_fr_l_MSTR"(
    ComponentPose=@n30
    Alpha=Get NonBackwardMovementAlpha() -> (NonBackwardMovementAlpha:double)
    Translation="0,0,2.5"
    Rotation="0,0,0"
    Scale="1,1,1"
   ) {
    rotationMode=BMM_Additive
    rotationSpace=BCS_BoneSpace
    translationMode=BMM_Additive
    translationSpace=BCS_BoneSpace
    alphaBoolBlend {blendOption=Linear}
    boneToModify {boneName=dyn_shirt_hem_fr_l_MSTR}
    translation {z=2.5}
   }
   n32 = Look At - Bone: dyn_shirt_hem_fr_r_MSTR(ComponentPose=@n31, Alpha=Get RightThighBendAlpha_Percentage() -> (RightThighBendAlpha_Percentage:double)) {
    alphaBoolBlend {blendOption=Linear}
    boneToModify {boneName=dyn_shirt_hem_fr_r_MSTR}
    lookAtLocation {x=-15, z=15}
    lookAtTarget {boneReference {boneName=calf_r}}
    lookAt_Axis {axis=<1 0>}
   }
   n33 = "Transform (Modify) Bone - Bone: dyn_shirt_hem_fr_r_MSTR"(ComponentPose=@n32, Alpha=@n22, Translation="-1.5,0,-1", Rotation="-40,0,5", Scale="1,1,1") {
    rotationMode=BMM_Additive
    rotationSpace=BCS_BoneSpace
    translationMode=BMM_Additive
    alphaBoolBlend {blendInTime=0.3, blendOption=ExpInOut, blendOutTime=0.3}
    boneToModify {boneName=dyn_shirt_hem_fr_r_MSTR}
    rotation {pitch=-40, roll=5}
    translation {x=-1.5, z=-1}
   }
   n34 = "Transform (Modify) Bone - Bone: dyn_shirt_hem_fr_r_2"(ComponentPose=@n33, Alpha=@n22, Translation="0,0,0", Rotation="0,0,0", Scale="1,1,1") {
    rotationMode=BMM_Additive
    rotationSpace=BCS_BoneSpace
    translationMode=BMM_Additive
    alphaBoolBlend {blendInTime=0.3, blendOption=ExpInOut, blendOutTime=0.3}
    boneToModify {boneName=dyn_shirt_hem_fr_r_2}
   }
   n35 = "Transform (Modify) Bone - Bone: right_leg_strap_MSTR"(ComponentPose=@n34, Alpha=@n22, Translation="0,-4,-1.5", Rotation="0,0,0", Scale="1,1,1") {
    translationMode=BMM_Additive
    alphaBoolBlend {blendInTime=0.3, blendOption=ExpInOut, blendOutTime=0.3}
    boneToModify {boneName=right_leg_strap_MSTR}
    translation {y=-4, z=-1.5}
   }
   n36 = "Constraint - dyn_buckle_MSTR (PPPPP)"!(
    ComponentPose=@n35
    Alpha=1
    ConstraintWeights_0=0.036
    ConstraintWeights_1:float=0.022
    ConstraintWeights_2:float=0.03
    ConstraintWeights_3:float=0.894
    ConstraintWeights_4:float=0.018
   ) -> (Pose) {
    alphaBoolBlend {blendOption=Linear}
    boneToModify {boneName=dyn_buckle_MSTR}
    constraintSetup [
     {targetBone {boneName=spine_02}}
     {targetBone {boneName=spine_03}}
     {targetBone {boneName=spine_01}}
     {targetBone {boneName=pelvis}}
     {targetBone {boneName=deform_groin_r}}
    ]
    constraintWeights [1,1,1,1,1]
   }
   n37 = Look At - Bone: dyn_buckle_MSTR(ComponentPose=@n36, Alpha=@n22) {
    alphaBoolBlend {blendOption=Linear}
    boneToModify {boneName=dyn_buckle_MSTR}
    lookAtLocation {x=-10, z=-2}
    lookAtTarget {boneReference {boneName=calf_r}}
    lookUp_Axis {axis=<0 1>}
   }
   n38 = "Transform (Modify) Bone - Bone: dyn_buckle_MSTR"(ComponentPose=@n37, Alpha=@n22, Translation="0,1,-3", Rotation="0,-100,60", Scale="1,1,1") {
    rotationMode=BMM_Additive
    rotationSpace=BCS_BoneSpace
    translationMode=BMM_Additive
    translationSpace=BCS_BoneSpace
    alphaBoolBlend {blendInTime=0.3, blendOption=ExpInOut, blendOutTime=0.3}
    boneToModify {boneName=dyn_buckle_MSTR}
    rotation {roll=60, yaw=-100}
    translation {y=1, z=-3}
   }
   n39 = "Transform (Modify) Bone - Bone: dyn_buckle_MSTR"(
    ComponentPose=@n38
    Alpha=Get BackwardMovementAlpha() -> (BackwardMovementAlpha:double)
    Translation="-1,-0.1,0"
    Rotation="0,0,0"
    Scale="1,1,1"
   ) {
    rotationMode=BMM_Additive
    rotationSpace=BCS_BoneSpace
    translationMode=BMM_Additive
    translationSpace=BCS_BoneSpace
    alphaBoolBlend {blendInTime=0.3, blendOption=ExpInOut, blendOutTime=0.3}
    boneToModify {boneName=dyn_buckle_MSTR}
    translation=<-1 -0.1>
   }
   n40 = "Transform (Modify) Bone - Bone: dyn_buckle_MSTR"(
    ComponentPose=@n39
    Alpha=Get ADS while standing Alpha() -> (ADS while standing Alpha:double)
    Translation="0,0,0"
    Rotation="-40,0,0"
    Scale="1,1,1"
   ) {
    rotationMode=BMM_Additive
    rotationSpace=BCS_BoneSpace
    translationSpace=BCS_BoneSpace
    alphaBoolBlend {blendOption=Linear}
    alphaScaleBiasClamp {bInterpResult=true, interpSpeedDecreasing=5, interpSpeedIncreasing=5}
    boneToModify {boneName=dyn_buckle_MSTR}
    rotation {pitch=-40}
   }
   n41 = Look At - Bone: dyn_belt_strap_MSTR(ComponentPose=@n40, Alpha=@n22) {
    alphaBoolBlend {blendOption=Linear}
    boneToModify {boneName=dyn_belt_strap_MSTR}
    lookAtLocation {x=-15, z=15}
    lookAtTarget {boneReference {boneName=calf_r}}
    lookAt_Axis {axis=<1 0>}
   }
   n42 = "Transform (Modify) Bone - Bone: dyn_belt_strap_MSTR"(ComponentPose=@n41, Alpha=@n22, Translation="0,2,0.5", Rotation="30,-10,0", Scale="1,1,1") {
    rotationMode=BMM_Additive
    rotationSpace=BCS_BoneSpace
    translationMode=BMM_Additive
    alphaBoolBlend {blendInTime=0.3, blendOption=ExpInOut, blendOutTime=0.3}
    boneToModify {boneName=dyn_belt_strap_MSTR}
    rotation {pitch=30, yaw=-10}
    translation {y=2, z=0.5}
   }
   n43 = "Transform (Modify) Bone - Bone: dyn_belt_strap_2"(ComponentPose=@n42, Alpha=@n22, Translation="0,3,0", Rotation="0,0,0", Scale="1,1,1") {
    rotationMode=BMM_Additive
    rotationSpace=BCS_BoneSpace
    translationMode=BMM_Additive
    alphaBoolBlend {blendInTime=0.3, blendOption=ExpInOut, blendOutTime=0.3}
    boneToModify {boneName=dyn_belt_strap_2}
    translation {y=3}
   }
   n44 = "Transform (Modify) Bone - Bone: deform_right_leg_pouch"(ComponentPose=@n43, Alpha=@n22, Translation="5,0,0", Rotation="-10,-60,-10", Scale="1,1,1") {
    rotationMode=BMM_Additive
    rotationSpace=BCS_BoneSpace
    translationMode=BMM_Additive
    translationSpace=BCS_BoneSpace
    alphaBoolBlend {blendOption=Linear}
    boneToModify {boneName=deform_right_leg_pouch}
    rotation=<rot -10 -60 -10>
    translation {x=5}
   }
   n45 = "Transform (Modify) Bone - Bone: right_leg_strap_MSTR"!(ComponentPose=@n44, bAlphaBoolEnabled:bool=True, Translation="0,0,0", Rotation="0,0,30", Scale="1,1,1") -> (Pose) {
    alphaInputType=Bool
    rotationMode=BMM_Additive
    translationMode=BMM_Additive
    alphaBoolBlend {blendInTime=0.3, blendOption=ExpInOut, blendOutTime=0.3}
    boneToModify {boneName=right_leg_strap_MSTR}
    rotation {roll=30}
   }
   n46 = Anim Dynamics - Bone: dyn_grenade_1_1(ComponentPose=@n45, Alpha=@n20, GravityOverride=@n21, ExternalForce="0,0,0") {
    angularBiasOverride=0.999
    angularDampingOverride=0.999
    angularSpringConstant=1000
    bAngularSpring=true
    bOverrideAngularDamping=true
    bUseGravityOverride=true
    simulationSpace=BoneRelative
    alphaBoolBlend {blendOption=Linear}
    boundBone {boneName=dyn_grenade_1_1}
    componentAppliedLinearAccClamp=<1000 1000 1000>
    componentLinearAccScale=<0.2 0.2 0.2>
    componentLinearVelScale=<0.2 0.2 0.2>
    physicsBodyDefinitions [{
      collisionType=CustomSphere
      sphereCollisionRadius=1
      boundBone {boneName=dyn_grenade_1_1}
      boxExtents=<40 60 60>
      constraintSetup {angularLimitsMax=<0 20 25>, angularLimitsMin=<0 -20 0>, angularTarget=<1 0 0>}
      localJointOffset=<-2 0 0>
     }]
    relativeSpaceBone {boneName=root}
   }
   n47 = Anim Dynamics - Bone: dyn_grenade_2_1(ComponentPose=@n46, Alpha=@n20, GravityOverride=@n21, ExternalForce="0,0,0") {
    angularBiasOverride=0.999
    angularDampingOverride=0.999
    angularSpringConstant=1000
    bAngularSpring=true
    bOverrideAngularDamping=true
    bUseGravityOverride=true
    linearSpringConstant=100
    simulationSpace=BoneRelative
    alphaBoolBlend {blendOption=Linear}
    boundBone {boneName=dyn_grenade_2_1}
    componentAppliedLinearAccClamp=<1000 1000 1000>
    componentLinearAccScale=<0.2 0.2 0.2>
    componentLinearVelScale=<0.2 0.2 0.2>
    physicsBodyDefinitions [{
      collisionType=CustomSphere
      sphereCollisionRadius=1
      boundBone {boneName=dyn_grenade_2_1}
      boxExtents=<30 50 50>
      constraintSetup {angularLimitsMax=<0 20 7>, angularLimitsMin=<0 -20 0>, angularTarget=<1 0 0>}
      localJointOffset=<-1.8 0 0>
     }]
    relativeSpaceBone {boneName=root}
   }
   n48 = Anim Dynamics - Bone: dyn_grenade_1_1(ComponentPose=@n47, Alpha=@n19, GravityOverride="0,0,-1500", ExternalForce="0,0,0") {
    angularDampingOverride=0.999
    angularSpringConstant=1000
    bAngularSpring=true
    bOverrideAngularDamping=true
    bUseGravityOverride=true
    simulationSpace=BoneRelative
    alphaBoolBlend {blendOption=Linear}
    boundBone {boneName=dyn_grenade_1_1}
    componentAppliedLinearAccClamp=<2000 2000 2000>
    componentLinearAccScale=<0.2 0.2 0.2>
    componentLinearVelScale=<0.2 0.2 0.2>
    physicsBodyDefinitions [{
      boundBone {boneName=dyn_grenade_1_1}
      boxExtents=<20 27 27>
      constraintSetup {angularLimitsMax=<20 20 5>, angularLimitsMin=<-20 -20 -5>, angularTarget=<1 0 0>}
      localJointOffset=<-2 0 0>
     }]
    relativeSpaceBone {boneName=root}
   }
   n49 = Anim Dynamics - Bone: dyn_grenade_2_1(ComponentPose=@n48, Alpha=@n19, GravityOverride="0,0,-1500", ExternalForce="0,0,0") {
    angularDampingOverride=0.999
    angularSpringConstant=1000
    bAngularSpring=true
    bOverrideAngularDamping=true
    bUseGravityOverride=true
    linearSpringConstant=100
    simulationSpace=BoneRelative
    alphaBoolBlend {blendOption=Linear}
    boundBone {boneName=dyn_grenade_2_1}
    componentAppliedLinearAccClamp=<2000 2000 2000>
    componentLinearAccScale=<0.2 0.2 0.2>
    componentLinearVelScale=<0.2 0.2 0.2>
    physicsBodyDefinitions [{
      boundBone {boneName=dyn_grenade_2_1}
      boxExtents=<20 30 30>
      constraintSetup {angularLimitsMax=<20 20 5>, angularLimitsMin=<-20 -20 -5>, angularTarget=<1 0 0>}
      localJointOffset=<-1.8 0 0>
     }]
    relativeSpaceBone {boneName=root}
   }
   n50 = Anim Dynamics - Bone: dyn_grenade_1_2(ComponentPose=@n49, Alpha=0.5, GravityOverride="0,0,-1500", ExternalForce="0,0,0") {
    angularDampingOverride=0.7
    angularSpringConstant=100
    bAngularSpring=true
    bOverrideAngularDamping=true
    bUseGravityOverride=true
    linearSpringConstant=100
    alphaBoolBlend {blendOption=Linear}
    boundBone {boneName=dyn_grenade_1_2}
    componentAppliedLinearAccClamp=<2000 2000 2000>
    componentLinearAccScale=<0.6 0.6 0.6>
    componentLinearVelScale=<0.6 0.6 0.6>
    gravityOverride {z=-50}
    physicsBodyDefinitions [{
      boundBone {boneName=dyn_grenade_1_2}
      boxExtents=<8 8 3>
      constraintSetup {angularLimitsMax=<10 0 10>, angularLimitsMin=<-10 -130 -10>, angularTarget=<1 0 0>}
      localJointOffset=<0.5 0 -1.5>
     }]
    relativeSpaceBone {boneName=dyn_grenade_1_1}
   }
   n51 = Anim Dynamics - Bone: dyn_grenade_2_2(ComponentPose=@n50, Alpha=0.7, GravityOverride="0,0,-1500", ExternalForce="0,0,0") {
    angularDampingOverride=0.7
    angularSpringConstant=100
    bAngularSpring=true
    bOverrideAngularDamping=true
    bUseGravityOverride=true
    linearSpringConstant=100
    alphaBoolBlend {blendOption=Linear}
    boundBone {boneName=dyn_grenade_2_2}
    componentAppliedLinearAccClamp=<2000 2000 2000>
    componentLinearAccScale=<0.6 0.6 0.6>
    componentLinearVelScale=<0.6 0.6 0.6>
    gravityOverride {z=-50}
    physicsBodyDefinitions [{
      boundBone {boneName=dyn_grenade_2_2}
      boxExtents=<8 8 3>
      constraintSetup {angularLimitsMax=<15 0 10>, angularLimitsMin=<-15 -130 -10>, angularTarget=<1 0 0>}
      localJointOffset=<0.5 0 -1.5>
     }]
    relativeSpaceBone {boneName=dyn_grenade_2_1}
   }
   n52 = Anim Dynamics - Bone: left_leg_zipper_1(ComponentPose=@n51, Alpha=1, GravityOverride="0,0,-300", ExternalForce="0,0,0") {
    angularDampingOverride=0.999
    angularSpringConstant=1000
    bAngularSpring=true
    bUseGravityOverride=true
    alphaBoolBlend {blendOption=Linear}
    boundBone {boneName=left_leg_zipper_1}
    componentAppliedLinearAccClamp=<1000 1000 1000>
    componentLinearAccScale=<0.2 0.2 0.2>
    componentLinearVelScale=<0.2 0.2 0.2>
    physicsBodyDefinitions [{
      boundBone {boneName=left_leg_zipper_1}
      boxExtents=<10 10 10>
      constraintSetup {angularLimitsMax=<0 0 5>, angularLimitsMin=<0 -90 -5>, angularTarget=<1 0 0>}
      localJointOffset=<2 0 1>
     }]
   }
   n53 = Anim Dynamics - Bone: dyn_carabiner_1(ComponentPose=@n52, Alpha=1, GravityOverride="0,0,-1960", ExternalForce="0,0,0") {
    angularDampingOverride=1
    angularSpringConstant=10000
    bAngularSpring=true
    bOverrideAngularDamping=true
    bUseGravityOverride=true
    alphaBoolBlend {blendOption=Linear}
    boundBone {boneName=dyn_carabiner_1}
    componentAppliedLinearAccClamp=<2000 2000 2000>
    componentLinearAccScale=<0.6 0.6 0.6>
    componentLinearVelScale=<0.6 0.6 0.6>
    physicsBodyDefinitions [{
      boundBone {boneName=dyn_carabiner_1}
      boxExtents=<20 40 40>
      constraintSetup {angularLimitsMax=<30 10 30>, angularLimitsMin=<-30 -140 -30>, angularTarget=<0 0 0>}
      localJointOffset=<-20 0 0>
     }]
   }
   n54 = "Constraint - dyn_front_strap_zipper_MSTR (PPPP)"(ComponentPose=@n53, Alpha=1, ConstraintWeights_0=0.023, ConstraintWeights_1=0.122, ConstraintWeights_2=0.475, ConstraintWeights_3=0.38) {
    alphaBoolBlend {blendOption=Linear}
    boneToModify {boneName=dyn_front_strap_zipper_MSTR}
    constraintSetup [{targetBone {boneName=spine_02}}, {targetBone {boneName=spine_03}}, {targetBone {boneName=spine_04}}, {targetBone {boneName=spine_05}}]
    constraintWeights [1,1,1,1]
   }
   n55 = "Constraint - dyn_carabiner_MSTR (PPPPPP)"(
    ComponentPose=@n54
    Alpha=1
    ConstraintWeights_0=0.139
    ConstraintWeights_1=0.332
    ConstraintWeights_2=0.367
    ConstraintWeights_3=0.126
    ConstraintWeights_4=0.018
    ConstraintWeights_5=0.018
   ) {
    alphaBoolBlend {blendOption=Linear}
    boneToModify {boneName=dyn_carabiner_MSTR}
    constraintSetup [
     {targetBone {boneName=spine_02}}
     {targetBone {boneName=spine_03}}
     {targetBone {boneName=spine_04}}
     {targetBone {boneName=spine_05}}
     {targetBone {boneName=spine_01}}
     {targetBone {boneName=clavicle_l}}
    ]
    constraintWeights [1,1,1,1,1,1]
   }
   n56 = Component To Local(ComponentPose=@n55)
   n57 = Get bRigidBodyEnabled()
   n58 = "Transform (Modify) Bone - Bone: dyn_front_strap_MSTR"!(
    ComponentPose=Local To Component(LocalPose=Blend Poses by bool(BlendPose_0=@n56, BlendPose_1=@n18, bActiveValue=@n57) {blendType=Linear})
    bAlphaBoolEnabled:bool=True
    Translation="0,0,0.5"
    Rotation="-3,0,0"
    Scale="1,1,1"
   ) -> (Pose) {
    alphaInputType=Bool
    rotationMode=BMM_Additive
    rotationSpace=BCS_BoneSpace
    translationMode=BMM_Additive
    translationSpace=BCS_BoneSpace
    alphaBoolBlend {blendOption=Sinusoidal}
    boneToModify {boneName=dyn_front_strap_MSTR}
    rotation {pitch=-3}
    translation {z=0.5}
   }
   Output Pose(Result=Component To Local(ComponentPose=@n58)) {layerGroup=None}
   Reroute Node(InputPin:Wildcard) -> (OutputPin:Wildcard)
  ]}
 EventGraph {nodes [
   n1 = Event BlueprintUpdateAnimation()
   > Set Delta Time X(Delta Time X=@n1.DeltaTimeX)
   > n2 = Sequence!(execute) -> (then_0, then_1, then_2:Exec, then_3:Exec, then_4:Exec)
   > Set LeftThighBendAlpha(LeftThighBendAlpha:double=GetAlphaFromReferenceTransformDelta(SkeletalMeshComponent=Get Skel Mesh(), BoneName=thigh_l, BaseName=pelvis, InputRange=(X=-17,Y=-103), TwistAxis="1,0,0")) -> (Output_Get:double)
   > n3 = Set RightThighBendAlpha(RightThighBendAlpha:double=GetAlphaFromReferenceTransformDelta(SkeletalMeshComponent=Get Skel Mesh(), BoneName=thigh_r, BaseName=pelvis, InputRange=(X=-17,Y=-103), TwistAxis="1,0,0")) -> (Output_Get:double)
   > Set RightThighBendAlpha_Percentage(RightThighBendAlpha_Percentage:double=float / float(A=@n3, B=1.5)) -> (Output_Get:double)
   n4 = OR Boolean(A=Get bIsCrouching(), B=Get bIsMoving2D())
   > Set ADS while standing Alpha(ADS while standing Alpha:double=SelectFloat(A=1, B=0, bPickA=NOT Boolean(A=@n4))) -> (Output_Get:double)
   Is Valid(exec=@n2.then_2, InputObject=Get Skel Mesh())
   > n5 = Gravity Override State Function DEPRECATED(CharacterPartAnimInstance=Self-Reference(), GameTimeInSeconds=GetGameTimeInSeconds(), GravityOverrideParams=Get Gravity Override Params())
   > Set Gravity Override(Gravity Override=@n5)
   Is Valid(exec=@n2.then_3, InputObject=Get PlayerPawn() -> (PlayerPawn:Fort Player Pawn Athena Object Reference))
   > n6 = Branch(Condition=Get bIsInFrontEnd())
   Set WPO Globals Function(
    execute=@n6.else
    Material=Get Mat Instance Dynamic()
    Overlay Material=Get Overlay Mat Instance Dynamic()
    Skeletal Mesh=Get Skel Mesh()
    "Local Noise Transform  Location"=Get Local Noise Transform Location() -> (Local Noise Transform Location:Vector)
    Local Noise Offset=Get Local Noise Offset()
   )
   > Set WPO Regional Function!(
    execute
    Material:Material Instance Dynamic Object Reference=Get Mat Instance Dynamic()
    Overlay Material:Material Instance Dynamic Object Reference=Get Overlay Mat Instance Dynamic()
    Skeletal Mesh:Skeletal Mesh Component Object Reference=Get Skel Mesh()
    Velocity Transform:Name=right_leg_strap_2
    Velocity Interp Speed:double=Get Velocity Interp Speed()
    Collision Transform:Name=calf_r
    Collision Transform Axis:EAxis Enum
    Invert Collision Transform Axis:bool=false
    Delta Time X:double=Get Delta Time X()
    Previous Velocity Transform Location:"Vector (by ref)"=Get R1 Velocity Transform Location()
    Previous Velocity Transform Velocity:"Vector (by ref)"=Get R1 Velocity Transform Velocity()
    Region:int=1
   ) -> (then)
   > n7 = Branch(Condition=Get bIsSkydiveFloating())
   > n8 = Set WPO Pho Wind Intensity(WPO Pho Wind Intensity=1)
   SetScalarParameterValue(execute=[@n9.then, @n8.then], self=Get Mat Instance Dynamic(), ParameterName=ClothSim_WindStrength, Value=Get WPO Pho Wind Intensity())
   n9 = Set WPO Pho Wind Intensity(execute=@n7.else, WPO Pho Wind Intensity=0)
   n10 = TryGetPawnOwner()
   n11 = GetVelocity!(self:Actor Object Reference=@n10) -> (ReturnValue:Vector)
   n12 = Dot Product!(A:Vector=@n11, B:Vector=GetActorForwardVector!(self:Actor Object Reference=@n10) -> (ReturnValue:Vector)) -> (ReturnValue:double)
   n13 = MapRangeClamped(Value=@n12, InRangeA=0, InRangeB=400)
   n14 = GetActorRightVector!(self:Actor Object Reference=@n10) -> (ReturnValue:Vector)
   n15 = MapRangeClamped(Value="Absolute (Float)"(A=Dot Product(A=@n11, B=@n14)), InRangeA=0, InRangeB=230)
   Set NonBackwardMovementAlpha(execute=@n2.then_1, NonBackwardMovementAlpha:double="Clamp (Float)"(Value=float + float(A=@n13, B=@n15), Min=0, Max=1)) -> (Output_Get:double)
   > Set BackwardMovementAlpha(BackwardMovementAlpha:double=MapRangeClamped(Value=@n12, InRangeA=0, InRangeB=-200)) -> (Output_Get:double)
   n16 = Get bIsSkyDiving()
   Set Skydiving(execute=@n2.then_4, Skydiving:double=SelectFloat(A=1, B=0, bPickA=@n16)) -> (Output_Get:double)
   > Set Not skydiving(Not skydiving:double=SelectFloat(A=1, B=0, bPickA=NOT Boolean(A=@n16))) -> (Output_Get:double)
   Event OnCharacterCustomizationLoaded()
   > n17 = Set Skel Mesh(Skel Mesh=GetPartSkeletalMeshComponent())
   > n18 = Get Dynamic Material Instance(SkeletalMeshComponent=@n17)
   > Set Mat Instance Dynamic(Mat Instance Dynamic=@n18)
   > n19 = GetOrCreateDynamicOverlayMaterial(MeshComponent=@n17, MaterialName=None)
   > Set Overlay Mat Instance Dynamic(Overlay Mat Instance Dynamic=@n19)
   Event BlueprintInitializeAnimation()
   > Set Skel Mesh(Skel Mesh=GetOwnerSkeletalMeshComponent())
   n20 = TryGetPawnOwner()
   > Is Valid(InputObject=@n20)
   > n21 = Cast To FortPlayerPawnAthena(Object=@n20)
   > Set PlayerPawn(PlayerPawn:Fort Player Pawn Athena Object Reference=@n21) -> (Output_Get:Fort Player Pawn Athena Object Reference)
  ]}
]
variables [
 ADS while standing Alpha {type=number}
 BackwardMovementAlpha {type=number}
 Delta Time X {type=number}
 ForwardMovementAlpha {type=number}
 Gravity Override {type=Vector}
 Gravity Override Params {
  type=GravityOverrideParamsStruct
  default {
   onGround_Falling {gravityOverride=<0 0 -50>}
   skydive_FloatingDown {gravityOverride=<0 -700 1200>, windAmplitude=<300 -500 400>, windFrequency=<3 10 10>}
  }
 }
 LeftThighBendAlpha {type=number}
 Local Noise Offset {type=Vector}
 Local Noise Transform Location {type=Vector}
 Mat Instance Dynamic {type=MaterialInstanceDynamic}
 NonBackwardMovementAlpha {type=number}
 Not skydiving {type=number}
 Overlay Mat Instance Dynamic {type=MaterialInstanceDynamic}
 PlayerPawn {type=FortPlayerPawnAthena}
 R1 Velocity Transform Location {type=Vector}
 R1 Velocity Transform Velocity {type=Vector}
 RightThighBendAlpha {type=number}
 RightThighBendAlpha_Percentage {type=number}
 SidewaysMovementAlpha {type=number}
 Skel Mesh {type=SkeletalMeshComponent}
 Skydiving {type=number}
 Velocity Interp Speed {type=number}
 WPO Pho Wind Intensity {type=number}
]
